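; Athena vehicle spawner for War Effort that spawns its vehicle only if the linked funding station is funded.
; On the authority it waits for match ready, checks that vehicles are enabled, applies any game mode class override, and skips spawning during the "NoVehicle" calendar event.
; Spawning supports a Y-axis-forward rotation and optional self-destruction afterward.
; Vehicle construction applies the forced-spawn setting and, when allowed, forced mods.
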(event EventSpawnVehicle
  (bind _returnvalue (CallFunction|GetVehicleClass))
  (bind _asactor (Utilities|Casting|CastToActorClass _returnvalue))
  (|CallParentFunction)
  (Networking|SwitchHasAuthority
    (:Authority
      (Class|FortAsyncActionAuthorityMatchReady|MatchReadyAsync
        (:OnReady
          (bind _returnvalue_2 (Class|BWarEffortFundingStationBase|IsFunded (Variables|Default|GetFundingStation) 0))
          (if _returnvalue_2
            (if (Fortnite|AreAthenaVehiclesEnabled)
              (bind _asfort_game_mode_athena (Utilities|Casting|CastToFortGameModeAthena (Game|GetGameMode))
                (:then
                  (if (Utilities|IsValidClass (Class|FortGameModeAthena|GetVehicleClassOverride _asfort_game_mode_athena _asactor))
                    _)
                  (bind _data (Utilities|FlowControl|MergeExecutionwithData (Class|FortGameModeAthena|GetVehicleClassOverride _asfort_game_mode_athena _asactor) _asactor)
                    (:Then
                      (bind _biseventactive (CalendarEvents|GetCalendarEventInformation "NoVehicle"))
                      (if (not _biseventactive)
                        (bind _returnvalue_3 (CallFunction|SpawnVehiclewithConstruction _data (select (Variables|Default|GetYAxisForward) (Math|Transform|ComposeTransforms (struct :Location "0, 0, 0" :Rotation "0,90.000000, 0" :Scale "1.000000,1.000000,1.000000") (Transformation|GetActorTransform)) (Transformation|GetActorTransform))))
                        (Variables|Default|SetSpawnedVehicle _returnvalue_3)
                        (if (Variables|Default|GetbShouldDestroyAfterVehicleSpawn)
                          (Actor|DestroyActor))))))
                (:CastFailed
                  _))
              (Actor|DestroyActor)
              (else
                _))
            (else
              _)))))
    (:Remote
      _)))

(event EventOnConstructVehicle
  (params (Vehicle :type Fort%20Athena%20Vehicle%20Object%20Reference))
  (|CallParentFunction Vehicle)
  (Spawn|SetForcedtoAlwaysSpawn Vehicle (CallFunction|GetForceSpawnAlways))
  (if (Variables|FortAthenaVehicleSpawner|GetbCanSpawnWithMod)
    (Vehicle|TryConstructwithModServer Vehicle (Variables|Default|GetForceMods))))

(fn ConstructionScript)
